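# Sound cue for the drop sound of the Core Sniper weapon, built from a single wave.
# The wave player feeds a modulator that slightly varies pitch on each play.
# The cue uses the shared weapon-drop attenuation and concurrency assets, the loot-collect sound class and a reduced volume.
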SoundCue /FlipperGameplay/Items/Weapons/CoreSniper/Audio/Core_Sniper_Drop_Cue
root=Modulator_1
nodes [
 {id=Modulator_1, children [WavePlayer_0], properties {pitchMax=1.1, pitchMin=0.9}}
 {id=WavePlayer_0, properties {soundWaveAssetPtr=&/Sniper_CoreSniper_Drop_01}}
]
properties {
 attenuationSettings=&Fort_Weapon_Drop_Att
 maxDistance=3000
 soundClassObject=&Fort_Loot_Collect
 volumeMultiplier=0.65
 concurrencyOverrides {resolutionRule=PreventNew}
 concurrencySet [&Weapon_Drop_StopOldest]
}
