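# First-person fire sound cue for the Core Sniper, layering the gunshot with a shell casing sound.
# A mixer blends one of three random close fire waves with a quieter, randomly delayed pick from six casing waves, varied in pitch and volume.
# A quality node supplies the casing layer only in its first two slots, leaving the other two empty.
# It uses first-person weapon attenuation and gunfire concurrency.
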SoundCue /FlipperGameplay/Items/Weapons/CoreSniper/Audio/Core_Sniper_Fire_1P_Cue
root=Mixer_0
nodes [
 {id=Delay_0, children [Modulator_1], properties {delayMax=0.55, delayMin=0.3}}
 {id=Mixer_0, children [Modulator_0, QualityLevel_0], properties {inputVolume [1.35,0.75]}}
 {id=Modulator_0, children [Random_0], properties {pitchMax=1, pitchMin=1, volumeMax=1, volumeMin=1}}
 {id=Modulator_1, children [Random_1], properties {pitchMax=1.1, pitchMin=0.9, volumeMax=1, volumeMin=0}}
 {id=QualityLevel_0, children [Delay_0,Delay_0,null,null]}
 {id=Random_0, children [WavePlayer_0,WavePlayer_1,WavePlayer_2], properties {bShouldExcludeFromBranchCulling=true, weights [1,1,1]}}
 {
  id=Random_1
  children [WavePlayer_3,WavePlayer_4,WavePlayer_5,WavePlayer_6,WavePlayer_7,WavePlayer_8]
  properties {bShouldExcludeFromBranchCulling=true, weights [1,1,1,1,1,1]}
 }
 {id=WavePlayer_0, properties {soundWaveAssetPtr=&/Sniper_CoreSniper_Fire_Close_1P_01}}
 {id=WavePlayer_1, properties {soundWaveAssetPtr=&/Sniper_CoreSniper_Fire_Close_1P_02}}
 {id=WavePlayer_2, properties {soundWaveAssetPtr=&/Sniper_CoreSniper_Fire_Close_1P_03}}
 {id=WavePlayer_3, properties {soundWaveAssetPtr=&/Sniper_CoreSniper_Shell_Casing_01}}
 {id=WavePlayer_4, properties {soundWaveAssetPtr=&/Sniper_CoreSniper_Shell_Casing_02}}
 {id=WavePlayer_5, properties {soundWaveAssetPtr=&/Sniper_CoreSniper_Shell_Casing_03}}
 {id=WavePlayer_6, properties {soundWaveAssetPtr=&/Sniper_CoreSniper_Shell_Casing_04}}
 {id=WavePlayer_7, properties {soundWaveAssetPtr=&/Sniper_CoreSniper_Shell_Casing_05}}
 {id=WavePlayer_8, properties {soundWaveAssetPtr=&/Sniper_CoreSniper_Shell_Casing_06}}
]
properties {
 attenuationSettings=&Fort_Weapon_Projectile_FirstPerson_Athena_Att
 bDebug=true
 maxDistance=2097152
 soundClassObject=&Fort_WeaponRanged_Fire_1P_PSM
 volumeMultiplier=1
 attenuationOverrides {manualReverbSendLevel=-1.0165, reverbSendMethod=Manual}
 concurrencyOverrides {bLimitToOwner=true, maxCount=2, resolutionRule=StopOldest, volumeScale=0.9}
 concurrencySet [&Weapon_Gunfire_FirstPerson_Conc]
}
