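# Sound cue for the Core Sniper gunshot, layering close transient, close surround tail, mid and distant sounds.
# Each layer randomly picks one of three waves, then passes through a modulator, its own distance attenuation and a sound class override before mixing.
# Close layers use the third-person fire class and mid and distant layers the distant class, while cue-level class, attenuation and concurrency are first-person.
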SoundCue /FlipperGameplay/Items/Weapons/CoreSniper/Audio/Core_Sniper_Fire_Cue
root=Mixer_17
nodes [
 {
  id=Attenuation_0
  children [Modulator_1]
  properties {attenuationSettings=&Fort_Weapon_Projectile_Medium_Mono_Athena_Att}
 }
 {id=Attenuation_1, children [Modulator_0], properties {attenuationSettings=/Game/Sounds/Attenuation/Fort_Weapon_Projectile_Close_Tails_Att}}
 {
  id=Attenuation_34
  children [Modulator_4]
  properties {attenuationSettings=&Fort_Weapon_Projectile_Distant_Mono_Athena_Att}
 }
 {id=Attenuation_49, children [Modulator_5], properties {attenuationSettings=&Fort_Weapon_Projectile_Close_Athena_Att}}
 {id=Mixer_0, children [SoundClass_32, SoundClass_1], properties {inputVolume [1,1]}}
 {id=Mixer_17, children [Mixer_0, SoundClass_0, SoundClass_35], properties {inputVolume [1.25,1.25,1.25]}}
 {id=Modulator_0, children [Random_3], properties {volumeMax=1, volumeMin=1}}
 {id=Modulator_1, children [Random_1], properties {volumeMax=1, volumeMin=1}}
 {id=Modulator_4, children [Random_8], properties {volumeMax=1, volumeMin=1}}
 {id=Modulator_5, children [Random_2], properties {volumeMax=1, volumeMin=1}}
 {id=Random_1, children [WavePlayer_12, WavePlayer_13, WavePlayer_14], properties {bShouldExcludeFromBranchCulling=true, weights [1,1,1]}}
 {id=Random_2, children [WavePlayer_0,WavePlayer_1,WavePlayer_2], properties {weights [1,1,1]}}
 {id=Random_3, children [WavePlayer_3,WavePlayer_4,WavePlayer_5], properties {bShouldExcludeFromBranchCulling=true, weights [1,1,1]}}
 {id=Random_8, children [WavePlayer_6,WavePlayer_7,WavePlayer_8], properties {bShouldExcludeFromBranchCulling=true, weights [1,1,1]}}
 {
  id=SoundClass_0
  children [Attenuation_0]
  properties {soundClassOverride=&Fort_WeaponRanged_Fire_Distant}
 }
 {
  id=SoundClass_1
  children [Attenuation_1]
  properties {soundClassOverride=&Fort_WeaponRanged_Fire_3P_PSM}
 }
 {
  id=SoundClass_32
  children [Attenuation_49]
  properties {soundClassOverride=&Fort_WeaponRanged_Fire_3P_PSM}
 }
 {
  id=SoundClass_35
  children [Attenuation_34]
  properties {soundClassOverride=&Fort_WeaponRanged_Fire_Distant}
 }
 {id=WavePlayer_0, properties {soundWaveAssetPtr=&/Sniper_CoreSniper_Fire_Close_Trans_03}}
 {id=WavePlayer_1, properties {soundWaveAssetPtr=&/Sniper_CoreSniper_Fire_Close_Trans_02}}
 {id=WavePlayer_12, properties {soundWaveAssetPtr=&/Sniper_CoreSniper_Fire_Mid_02}}
 {id=WavePlayer_13, properties {soundWaveAssetPtr=&/Sniper_CoreSniper_Fire_Mid_03}}
 {id=WavePlayer_14, properties {soundWaveAssetPtr=&/Sniper_CoreSniper_Fire_Mid_01}}
 {id=WavePlayer_2, properties {soundWaveAssetPtr=&/Sniper_CoreSniper_Fire_Close_Trans_01}}
 {id=WavePlayer_3, properties {soundWaveAssetPtr=&/Sniper_CoreSniper_Fire_Close_Tail_Srrnd_03}}
 {id=WavePlayer_4, properties {soundWaveAssetPtr=&/Sniper_CoreSniper_Fire_Close_Tail_Srrnd_02}}
 {id=WavePlayer_5, properties {soundWaveAssetPtr=&/Sniper_CoreSniper_Fire_Close_Tail_Srrnd_01}}
 {id=WavePlayer_6, properties {soundWaveAssetPtr=&/Sniper_CoreSniper_Fire_Distant_01}}
 {id=WavePlayer_7, properties {soundWaveAssetPtr=&/Sniper_CoreSniper_Fire_Distant_02}}
 {id=WavePlayer_8, properties {soundWaveAssetPtr=&/Sniper_CoreSniper_Fire_Distant_03}}
]
properties {
 attenuationSettings=&Fort_Weapon_Projectile_FirstPerson_Athena_Att
 bDebug=true
 maxDistance=2097152
 soundClassObject=&Fort_WeaponRanged_Fire_1P_PSM
 volumeMultiplier=0.9
 concurrencyOverrides {bLimitToOwner=true, maxCount=2, resolutionRule=StopOldest}
 concurrencySet [&Weapon_Gunfire_FirstPerson_Conc]
}
