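# Sound cue for the Core Sniper's low-ammo fire sound, routed to a degraded low-ammo ranged weapon sound class.
# A single wave plays after a short fixed delay, with a modulator raising its pitch by a narrowly randomized amount.
# It uses close-range weapon projectile attenuation and overrides concurrency to prevent new instances once the limit is reached.
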SoundCue /FlipperGameplay/Items/Weapons/CoreSniper/Audio/Core_Sniper_Fire_Low_Ammo_Cue
root=Delay_0
nodes [
 {id=Delay_0, children [Modulator_0], properties {delayMax=0.075, delayMin=0.075}}
 {id=Modulator_0, children [WavePlayer_0], properties {pitchMax=1.27, pitchMin=1.25, volumeMax=1, volumeMin=1}}
 {id=WavePlayer_0, properties {soundWaveAssetPtr=&/Sniper_CoreSniper_Fire_Low_Ammo_01}}
]
properties {
 attenuationSettings=&Fort_Weapon_Projectile_Close_Athena_Att
 bOverrideConcurrency=true
 maxDistance=4950
 soundClassObject=/Game/Sounds/FortSoundClassesAndModes/Fort_WeaponRanged_LowAmmo_Degraded_PSM
 volumeMultiplier=1.1
 concurrencyOverrides {resolutionRule=PreventNew}
}
