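# Sound cue for the dry-fire click of the Core Sniper when it is out of ammo.
# A single wave player plays the sniper dry-fire sample through a modulator that adds slight downward pitch variation.
# It runs at reduced volume on the first-person ranged-weapon fire sound class with close projectile attenuation.
# It uses the first-person gunfire concurrency set, with overrides that cap instances and stop the oldest.
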SoundCue /FlipperGameplay/Items/Weapons/CoreSniper/Audio/Core_Sniper_OutofAmmo_Cue
root=Modulator_0
nodes [
 {id=Modulator_0, children [WavePlayer_4], properties {pitchMax=1, pitchMin=0.98}}
 {id=WavePlayer_4, properties {soundWaveAssetPtr=&/Sniper_CoreSniper_Fire_Dry_01}}
]
properties {
 attenuationSettings=&Fort_Weapon_Projectile_Close_Athena_Att
 maxDistance=4950
 soundClassObject=&Fort_WeaponRanged_Fire_1P_PSM
 volumeMultiplier=0.35
 attenuationOverrides {falloffDistance=1600}
 concurrencyOverrides {maxCount=2, resolutionRule=StopOldest}
 concurrencySet [&Weapon_Gunfire_FirstPerson_Conc]
}
