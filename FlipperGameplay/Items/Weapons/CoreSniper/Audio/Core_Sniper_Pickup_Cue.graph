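# Sound cue for picking up the Core Sniper weapon, categorized under the loot-collect sound class.
# A single pickup wave plays through a Modulator node at reduced volume, using the shared weapon pickup attenuation.
# The cue references the shared PickupMaximum_Oldest concurrency set and also carries single-instance, prevent-new concurrency override settings.
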SoundCue /FlipperGameplay/Items/Weapons/CoreSniper/Audio/Core_Sniper_Pickup_Cue
root=Modulator_1
nodes [
 {id=Modulator_1, children [WavePlayer_2]}
 {id=WavePlayer_2, properties {soundWaveAssetPtr=&/Sniper_CoreSniper_Pickup_01}}
]
properties {
 attenuationSettings=&Fort_Weapon_Pickup_Att
 maxDistance=3500
 soundClassObject=&Fort_Loot_Collect
 volumeMultiplier=0.5077
 concurrencyOverrides {maxCount=1, resolutionRule=PreventNew}
 concurrencySet [&PickupMaximum_Oldest]
}
