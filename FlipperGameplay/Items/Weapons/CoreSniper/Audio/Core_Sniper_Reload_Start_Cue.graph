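# Sound cue for the start of the Core Sniper weapon's reload.
# A single wave player plays one reload-start sound wave directly as the root, with no randomization or mixing.
# It uses the shared weapon reload attenuation and weapon foley sound class at a reduced volume.
# The cue overrides concurrency and has its debug flag enabled.
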SoundCue /FlipperGameplay/Items/Weapons/CoreSniper/Audio/Core_Sniper_Reload_Start_Cue
root=WavePlayer_29
nodes [{id=WavePlayer_29, properties {soundWaveAssetPtr=&/Sniper_CoreSniper_Reload_Start_01}}]
properties {
 attenuationSettings=&Fort_Weapon_Reload_Att
 bDebug=true
 bOverrideConcurrency=true
 maxDistance=2250
 soundClassObject=&Fort_WeaponFoley
 volumeMultiplier=0.401
}
