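# Sound cue that plays the Core Sniper scope zoom-out sound, part of the weapon's audio.
# A single wave player feeds the zoom-out wave through a Modulator node at the root.
# It routes to the weapon foley sound class at reduced volume, with overridden attenuation settings.
# A concurrency override prevents new instances from starting once the limit is reached.
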SoundCue /FlipperGameplay/Items/Weapons/CoreSniper/Audio/Core_Sniper_Scope_ZoomOut_Cue
root=Modulator_0
nodes [
 {id=Modulator_0, children [WavePlayer_0]}
 {id=WavePlayer_0, properties {soundWaveAssetPtr=&/Sniper_CoreSniper_Scope_Zoom_Out_01}}
]
properties {
 bOverrideConcurrency=true
 maxDistance=2097152
 soundClassObject=&Fort_WeaponFoley
 volumeMultiplier=0.175
 attenuationOverrides {falloffDistance=1500, nonSpatializedRadiusStart=200, attenuationShapeExtents {x=1200}}
 concurrencyOverrides {resolutionRule=PreventNew}
}
