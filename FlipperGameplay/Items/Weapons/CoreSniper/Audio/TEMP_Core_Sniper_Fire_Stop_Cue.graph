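# Placeholder SoundCue, named as a temporary fire-stop sound for the Core Sniper weapon.
# It has no root or sound nodes, so it produces no audio.
# It still carries first-person ranged-fire routing through its sound class, projectile attenuation and gunfire concurrency settings, with reduced volume and the debug flag enabled.
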SoundCue /FlipperGameplay/Items/Weapons/CoreSniper/Audio/TEMP_Core_Sniper_Fire_Stop_Cue
root=null
nodes []
properties {
 attenuationSettings=&Fort_Weapon_Projectile_FirstPerson_Athena_Att
 bDebug=true
 maxDistance=2097152
 soundClassObject=&Fort_WeaponRanged_Fire_1P_PSM
 volumeMultiplier=0.7
 concurrencySet [&Weapon_Gunfire_FirstPerson_Conc]
}
