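# Niagara one-shot effect for the default bullet impact of the Core Sniper weapon.
# Independent burst emitters layer a hot flash with light, center and radial spikes, an airflow mesh, eroding smoke, and falling mesh and flipbook debris.
# Distance-based LOD remaps enlarge most elements and speed up debris and spikes to keep distant hits readable.
# One random distribution value per debris mesh couples size, speed and gravity, so larger chunks launch slower and fall harder.
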NiagaraSystem /FlipperGameplay/Items/Weapons/CoreSniper/FX/NS_Core_Sniper_Impact_Default
dependencies {
 usedDataInterfaces [
  &NiagaraDataInterfaceColorCurve
  &NiagaraDataInterfaceCurve
  &NiagaraDataInterfaceVectorCurve
  &NiagaraDataInterfaceVector2DCurve
 ]
 usedDynamicInputs []
 usedModules [&SystemState]
}
emitters {
 Airflow {
  emitterUpdateScript [
   &EmitterState()
   &SpawnBurst_Instantaneous(Spawn Count=1, Use Spawn Probability=false)
  ]
  particleSpawnScript [
   &InitializeParticle(
    Color=<rgba 0.4318 0.4307 0.4375 1.5>
    Lifetime Max=0.4
    Lifetime Min=0.35
    Lifetime Mode=::NewEnumerator1
    Mesh Scale Mode=::NewEnumerator1
    Mesh Uniform Scale Max=2.75
    Mesh Uniform Scale Min=2
    Sprite Size Mode=::NewEnumerator1
    Uniform Sprite Size Max=2.75
    Uniform Sprite Size Min=2
    UsePositionOffset=false
    Write Material Random=true
   ) {module=InitializeParticle001}
   &CameraOffset(Camera Offset Amount=40)
   &AddVelocity(Velocity=&RandomRangeVector(Minimum=<5 0 0>, Maximum=<10 1 1>))
   &InitialMeshOrientation(Mesh Orientation Mode=::NewEnumerator2, Rotation=<0 0.25 0>, Use Rotation Vector=true)
  ]
  particleUpdateScript [
   &ParticleState()
   &ScaleColor(
    Scale Alpha=&FloatFromCurve(
     FloatCurve=DataInterface() {propertyValues {Curve {keys [<at 0 0>,<at 0.1 0.8>,<at 1 0>]}}}
     Scale Curve=&RandomRangeFloat(Minimum=15, Maximum=30)
    )
    Scale RGB=&MakeVectorFromLinearColorRGB(Color=&ColorFromCurve(ColorCurve=DataInterface() {propertyValues {
        AlphaCurve {keys []}
        BlueCurve {keys [
          {arriveTangent=0.0185, interpMode=RCIM_Cubic, leaveTangent=0.0185, time=0, value=0.0961}
          <at 0.4191 0.8>
          {arriveTangent=0.0185, interpMode=RCIM_Cubic, leaveTangent=0.0185, time=1, value=0.1146}
         ]}
        GreenCurve {keys [
          {arriveTangent=-0.7509, interpMode=RCIM_Cubic, leaveTangent=-0.7509, time=0, value=0.8322}
          <at 0.4191 0.8>
          {arriveTangent=-0.7509, interpMode=RCIM_Cubic, leaveTangent=-0.7509, time=1, value=0.0813}
         ]}
        RedCurve {keys [
          {arriveTangent=-1.9276, interpMode=RCIM_Cubic, leaveTangent=-1.9276, time=0, value=2}
          <at 0.4191 0.8>
          {arriveTangent=-1.9276, interpMode=RCIM_Cubic, leaveTangent=-1.9276, time=1, value=0.0724}
         ]}
       }}))
    ScaleA=true
    ScaleRGB=true
   )
   &ScaleMeshSize(Scale Factor=&VectorFromCurve(
     VectorCurve=DataInterface() {propertyValues {XCurve {keys [<at 0 1>,<at 1 1>]}, YCurve {keys [<at 0 1>,<at 1 1>]}, ZCurve {keys [<at 0 0.2>,<at 1 1>]}}}
     Scale Curve=&VectorFromFloat(Value=&ClampFloat(
       Float=&FN_LOD_Distance_Remap(Start Distance=0, End Distance=8000, Minimum Output=1, Maximum Output=3)
       Min=1
       Max=4
      ))
    ))
   &SolveForcesAndVelocity(Clamp Velocity=false, Limit Acceleration=false)
   &DynamicMaterialParameters(
    Index 0 Param 1=&FloatFromCurve(FloatCurve=DataInterface() {propertyValues {Curve {keys [{tangentMode=RCTM_User, time=0, value=0}, {tangentMode=RCTM_User, time=1, value=1}]}}})
    Index 0 Param 2=&FloatFromCurve(FloatCurve=DataInterface() {propertyValues {Curve {keys [{tangentMode=RCTM_User, time=0, value=0.5}, {tangentMode=RCTM_User, time=1, value=1.5}]}}})
    Index 0 Param 3=1
    Index 0 Param 4=1
    Module.Param0WriteEnabled=true
    Module.Param1WriteEnabled=true
    Module.Param2WriteEnabled=true
    Module.Param3WriteEnabled=true
   )
  ]
  rendererClasses [&NiagaraMeshRendererProperties]
  renderers [{rendererClass=&NiagaraMeshRendererProperties, rendererIndex=0}]
 }
 Debris_Mesh_BETA {
  emitterUpdateScript [
   &EmitterState()
   &SpawnBurst_Instantaneous(Spawn Count=&RandomRangeInt(Minimum=7, Maximum=15), Use Spawn Probability=false)
  ]
  particleSpawnScript [
   /SetVariables(Particles.Distribution:NiagaraFloat=&Add_Float(
     A=&UniformAOrBFloat(A=0.25, B=0.9, Distribution Weight=0.1)
     B=&RandomRangeFloat(Minimum=-0.1, Maximum=0.1)
    )) {isSetParameters=true}
   &InitializeParticle(
    Color=<rgba 0.0125 0.0099 0.0068 1>
    Lifetime Max=1.5
    Lifetime Mode=::NewEnumerator1
    Mesh Scale=&VectorFromFloat(Value=/Game/Effects/Niagara/DynamicInputs/Math/FN_Fit01(Clamp:NiagaraBool=false, Value:NiagaraFloat=$Particles.Distribution, DestMin:NiagaraFloat=0.02, DestMax:NiagaraFloat=0.08))
    Mesh Scale Mode=::NewEnumerator4
    UsePositionOffset=false
    Write Material Random=true
   ) {module=InitializeParticle001}
   &InitialMeshOrientation(Orientation Vector=&RandomVector(Vector Scale=1), Use Orientation Vector=true, Use Rotation Vector=false)
   &AddVelocityInCone(
    Cone Angle=30
    Module.Use Velocity Falloff On Cone Axis=false
    Velocity Strength=&Multiply_Float(
     A=&FN_Fit01(
      Clamp:NiagaraBool=false
      Value:NiagaraFloat=&OneMinusFloat(Float=$Particles.Distribution)
      DestMin:NiagaraFloat=300
      DestMax:NiagaraFloat=684.8762
     )
     B=&ClampFloat(
      Float=&FN_LOD_Distance_Remap(Start Distance=0, End Distance=8000, Minimum Output=1, Maximum Output=1.2)
      Min=1
      Max=1.2
     )
    )
   )
   &ApplyInitialForces()
  ]
  particleUpdateScript [
   &ParticleState()
   &AccelerationForce(Acceleration=&MakeVector(
     X=0
     Y=0
     Z=&FN_Fit01(Clamp:NiagaraBool=false, Value:NiagaraFloat=$Particles.Distribution, DestMin:NiagaraFloat=-980, DestMax:NiagaraFloat=-1200)
    ))
   &Drag(Drag=0.5, Use Linear Drag=true, Use Rotational Drag=false)
   &SolveForcesAndVelocity(Clamp Velocity=false, Limit Acceleration=false)
   &SolveRotationalForcesAndVelocity()
   &FN_Mesh_Attributes(
    Create Scale Attribute=true
    Create Visibility Attribute=true
    Scale=&Multiply_Vector(
     A=$Particles.Initial.Scale
     B=&VectorFromFloat(Value=&FloatFromCurve(FloatCurve=DataInterface() {propertyValues {Curve {keys [
           <at -0.0032 0 RCIM_Cubic>
           {arriveTangent=-0.0978, interpMode=RCIM_Cubic, leaveTangent=-0.0978, tangentMode=RCTM_User, time=0.1249, value=0.9962}
           {arriveTangent=0.0267, interpMode=RCIM_Cubic, leaveTangent=0.0267, tangentMode=RCTM_User, time=0.7407, value=1.0021}
           <at 1 0 RCIM_Cubic>
          ]}}}))
    )
   )
   &ScaleMeshSize(Scale Factor=&VectorFromFloat(Value=&FloatFromCurve(
      FloatCurve=DataInterface() {propertyValues {Curve {keys [<at 0 0>,<at 0.05 1>,<at 0.7 1>,<at 1 0>]}}}
      Scale Curve=&ClampFloat(
       Float=&FN_LOD_Distance_Remap(Start Distance=0, End Distance=8000, Minimum Output=1, Maximum Output=2)
       Min=1
       Max=2
      )
     )))
   &UpdateMeshOrientation(
    Delta Time:NiagaraFloat=$Particles.Lifetime
    Rotation Rate=0.07
    Rotation Vector=&RandomRangeVector(Minimum=<-1 -1 -1>, Maximum=<1 1 1>)
   )
  ]
  rendererClasses [&NiagaraMeshRendererProperties]
  renderers [{rendererClass=&NiagaraMeshRendererProperties, rendererIndex=0}]
 }
 Debris_Sprite_SubUV {
  emitterUpdateScript [
   &EmitterState()
   &SpawnBurst_Instantaneous(Spawn Count=&RandomRangeInt(Minimum=6, Maximum=10), Use Spawn Probability=false)
  ]
  particleSpawnScript [
   &InitializeParticle(
    Color=<rgba 0.0125 0.0099 0.0068 1>
    Lifetime Max=1.5
    Lifetime Mode=::NewEnumerator1
    Sprite Size=&UniformRangedVector2D(Minimum=<2.1064 4.7118>, Maximum=<4.4317 8.5012>)
    Sprite Size Mode=::NewEnumerator4
    UsePositionOffset=false
    Write Material Random=true
   ) {module=InitializeParticle001}
   &AddVelocityInCone(
    Module.Use Velocity Falloff On Cone Axis=false
    Velocity Strength=&Multiply_Float(
     A=&UniformRangedFloat(Minimum=400, Maximum=600)
     B=&ClampFloat(
      Float=&FN_LOD_Distance_Remap(Start Distance=0, End Distance=8000, Minimum Output=1, Maximum Output=1.2)
      Min=1
      Max=1.2
     )
    )
   )
  ]
  particleUpdateScript [
   &ParticleState()
   &AccelerationForce(Acceleration=<0 0 -1200>)
   &Drag(Drag=0.1, Use Linear Drag=true, Use Rotational Drag=false)
   &SolveForcesAndVelocity(Clamp Velocity=false, Limit Acceleration=false)
   &ScaleSpriteSize(Scale Factor=&Vector2DFromFloat(Value=&FloatFromCurve(
      FloatCurve=DataInterface() {propertyValues {Curve {keys [<at 0.7424 1 RCIM_Cubic>,<at 1.0011 0.002>]}}}
      Scale Curve=&ClampFloat(
       Float=&FN_LOD_Distance_Remap(Start Distance=0, End Distance=8000, Minimum Output=1, Maximum Output=2)
       Min=1
       Max=2
      )
     )))
   &SubUVAnimation(End Frame=4, SubUV Animation Mode:ENiagara_SubUVLookupMode=::NewEnumerator2) {module=SubUVAnimation001}
   &Color(Color=$Particles.Initial.Color)
  ]
  rendererClasses [&NiagaraSpriteRendererProperties]
  renderers [{rendererClass=&NiagaraSpriteRendererProperties}]
 }
 Flash {
  emitterUpdateScript [
   &EmitterState()
   &SpawnBurst_Instantaneous(Spawn Count=1, Use Spawn Probability=false)
  ]
  particleSpawnScript [
   /Niagara/Modules/Spawn/Initialization/InitializeParticle(
    Lifetime=&UniformRangedFloat(Minimum=0.15, Maximum=0.15)
    Mesh Scale=&UniformRangedVector(Minimum=<0.5 0.7 0.7>, Maximum=<1 1.1 1.1>)
    Module.Write Color=true
    Module.Write Lifetime=true
    Module.Write Mass=true
    Module.Write Material Random=true
    Module.Write Position=true
    Module.Write Scale=true
    Module.Write SpriteRotation=false
    Module.Write SpriteSize=true
    Position=&Add_Vector(A=/Niagara/DynamicInputs/Helpers/SimulationPosition(), B=<0 0 0>)
    Sprite Size=<100 100>
   )
   /Game/Effects/Niagara/Modules/Orientation/FN_Fan(Spin:NiagaraFloat=&UniformRangedFloat(Minimum=0, Maximum=1))
   /SetVariables(Particles.LightPositionNew:Vector3f=&Add_Vector(
     A=Linked() {linkedVariable {name=Particles.Position, type=NiagaraPosition}}
     B=/Niagara/DynamicInputs/Transforms/TransformPosition(Position=<25 0 0>, Source Space=::Local, Destination Space=::Simulation)
    )) {isSetParameters=true, module=SetVariables_#1}
   &SubUVAnimation(End Frame=3, SubUV Animation Mode:ENiagara_SubUVLookupMode=::NewEnumerator2)
   &CameraOffset(Camera Offset Amount=10)
   &InitialMeshOrientation(
    Mesh Orientation Mode=::NewEnumerator2
    Rotation=&RandomRangeVector(Minimum=<0 0 0>, Maximum=<1 0 0>)
    Use Rotation Vector=true
   )
  ]
  particleUpdateScript [
   &ParticleState()
   &Color(
    Color=&MakeLinearColorFromVectorAndFloat(
     "Vector (RGB)"=&MakeVectorFromLinearColorRGB(Color=&ColorFromCurve(ColorCurve=DataInterface() {propertyValues {
         AlphaCurve {keys []}
         BlueCurve {keys [
           {arriveTangent=0.0059, interpMode=RCIM_Cubic, leaveTangent=0.0059, time=0, value=0}
           {arriveTangent=0.0059, interpMode=RCIM_Cubic, leaveTangent=0.0059, time=0.9977, value=0.0059}
          ]}
         GreenCurve {keys [
           {arriveTangent=-4.9082, interpMode=RCIM_Cubic, leaveTangent=-4.9082, time=0, value=4.903}
           {arriveTangent=-4.9082, interpMode=RCIM_Cubic, leaveTangent=-4.9082, time=0.9977, value=0.0059}
          ]}
         RedCurve {keys [
           {arriveTangent=-14.4701, interpMode=RCIM_Cubic, leaveTangent=-14.4701, time=0, value=15}
           {arriveTangent=-14.4701, interpMode=RCIM_Cubic, leaveTangent=-14.4701, time=0.9977, value=0.5625}
          ]}
        }}))
     "Float (Alpha)"=&FloatFromCurve(FloatCurve=DataInterface() {propertyValues {Curve {keys [{interpMode=RCIM_Cubic, tangentMode=RCTM_User, time=0, value=1}, {interpMode=RCIM_Cubic, tangentMode=RCTM_User, time=1, value=0}]}}})
    )
    Scale Color=&VectorFromFloat(Value=&FloatFromCurve(FloatCurve=DataInterface() {propertyValues {Curve {keys [{interpMode=RCIM_Cubic, tangentMode=RCTM_User, time=0, value=1}, {interpMode=RCIM_Cubic, tangentMode=RCTM_User, time=1, value=0}]}}}))
   )
   &ScaleMeshSize(Scale Factor=&VectorFromFloat(Value=&FloatFromCurve(
      FloatCurve=DataInterface() {propertyValues {Curve {keys [{arriveTangent=2.2457, interpMode=RCIM_Cubic, leaveTangent=2.2457, tangentMode=RCTM_User, time=0, value=0}, <at 1 1>]}}}
      Scale Curve=&ClampFloat(
       Float=&FN_LOD_Distance_Remap(Start Distance=0, End Distance=8000, Minimum Output=1.5, Maximum Output=3)
       Min=1.5
       Max=4
      )
     )))
   &FN_Sprite_Attributes(
    Create SpriteSize Attribute=true
    SpriteSize=&Multiply_Vector2D_ByFloat(
     Float=&FloatFromCurve(
      FloatCurve=DataInterface() {propertyValues {Curve {keys [
          {arriveTangent=4.1785, interpMode=RCIM_Cubic, leaveTangent=4.1785, tangentMode=RCTM_User, time=0, value=0}
          {arriveTangent=0.9995, interpMode=RCIM_Cubic, leaveTangent=0.9995, tangentMode=RCTM_User, time=0.2223, value=0.6825}
          {arriveTangent=-0.0405, interpMode=RCIM_Cubic, leaveTangent=-0.0405, tangentMode=RCTM_User, time=1, value=1}
         ]}}}
      Scale Curve=&Multiply_Float(
       A=1.3
       B=&ClampFloat(
        Float=&FN_LOD_Distance_Remap(Start Distance=0, End Distance=8000, Minimum Output=1.3, Maximum Output=1)
        Min=1.3
       )
      )
     )
     Vector2D=&Vector2DFromFloat(Value=&RandomRangeFloat(Minimum=90, Maximum=120))
    )
    Write to Module Namespace=true
   ) {module=FN_Sprite}
   /SetVariables(Particles.SpriteColor=&MakeLinearColorFromVectorAndFloat(
     "Vector (RGB)"=&MakeVectorFromLinearColorRGB(Color=&ColorFromCurve(ColorCurve=DataInterface() {propertyValues {
         AlphaCurve {keys []}
         BlueCurve {keys [
           {arriveTangent=-0.238, interpMode=RCIM_Cubic, leaveTangent=-0.238, time=0, value=0.4871}
           {arriveTangent=-0.238, interpMode=RCIM_Cubic, leaveTangent=-0.238, time=1, value=0.2491}
          ]}
         GreenCurve {keys [
           {arriveTangent=-2.5716, interpMode=RCIM_Cubic, leaveTangent=-2.5716, time=0, value=2.8407}
           {arriveTangent=-2.5716, interpMode=RCIM_Cubic, leaveTangent=-2.5716, time=1, value=0.2691}
          ]}
         RedCurve {keys [
           {arriveTangent=-6.5, interpMode=RCIM_Cubic, leaveTangent=-6.5, time=0, value=7}
           {arriveTangent=-6.5, interpMode=RCIM_Cubic, leaveTangent=-6.5, time=1, value=0.5}
          ]}
        }}))
     "Float (Alpha)"=&FloatFromCurve(FloatCurve=DataInterface() {propertyValues {Curve {keys [<at 0 1>,<at 1 0>]}}})
    )) {isSetParameters=true}
   &DynamicMaterialParameters(Write Parameter Index 0=false)
  ]
  rendererClasses [&NiagaraMeshRendererProperties, &NiagaraSpriteRendererProperties]
  renderers [
   {rendererClass=&NiagaraMeshRendererProperties, rendererIndex=0}
   {rendererClass=&NiagaraSpriteRendererProperties, rendererIndex=1}
  ]
 }
 Light {
  emitterUpdateScript [
   &EmitterState()
   &SpawnBurst_Instantaneous(Spawn Count=1, Use Spawn Probability=false)
  ]
  particleSpawnScript [
   /Niagara/Modules/Spawn/Initialization/V2/InitializeParticle(Lifetime=0.17, UsePositionOffset=false, Write Material Random=true) {module=InitializeParticle001}
   /SetVariables(Particles.LightPositionNew:Vector3f=&Add_Vector(
     A=Linked() {linkedVariable {name=Particles.Position, type=NiagaraPosition}}
     B=&TransformPosition(Position=<25 0 0>, Source Space=::Local, Destination Space=::Simulation)
    )) {isSetParameters=true, module=SetVariables_#1}
   /SetVariables(Particles.LightExponent=3) {isSetParameters=true}
  ]
  particleUpdateScript [
   /SetVariables(Particles.Light_Color:LinearColor=&MultiplyLinearColor(
     Linear Color=<rgba 1 0.6896 0.2104 1>
     Scale Factor=&MakeVector4_FromFloat(Float=&Multiply_Float(
       A=20
       B=&FloatFromCurve(FloatCurve=DataInterface() {propertyValues {Curve {keys [
            {interpMode=RCIM_Cubic, tangentMode=RCTM_User, time=0, value=0}
            {interpMode=RCIM_Cubic, tangentMode=RCTM_User, time=0.2, value=1}
            {tangentMode=RCTM_User, time=1, value=0}
           ]}}})
      ))
    )) {isSetParameters=true}
   /SetVariables(Particles.LightRadius=&Multiply_Float(
     A=40
     B=&ClampFloat(
      Float=&FN_LOD_Distance_Remap(Start Distance=0, End Distance=8000, Minimum Output=1, Maximum Output=2)
      Min=1
      Max=2
     )
    )) {isSetParameters=true}
   &ParticleState()
  ]
  rendererClasses [&NiagaraLightRendererProperties]
  renderers [{rendererClass=&NiagaraLightRendererProperties, rendererIndex=0}]
 }
 Smoke {
  emitterUpdateScript [
   &EmitterState()
   &SpawnBurst_Instantaneous(Spawn Count=&UniformRangedInt(Minimum=4, Maximum=6), Use Spawn Probability=false)
  ]
  particleSpawnScript [
   &InitializeParticle(
    Color=<rgba 0.2812 0.2812 0.2812 2.2>
    Lifetime Max=0.5
    Lifetime Min=0.3
    Lifetime Mode=::NewEnumerator1
    Sprite Rotation Mode=::NewEnumerator2
    Sprite Size Mode=::NewEnumerator1
    Uniform Sprite Size Max=80
    Uniform Sprite Size Min=50
    UsePositionOffset=false
    Write Material Random=true
   ) {module=InitializeParticle001}
   /SetVariables(Particles.InitSpriteRotRate=&UniformRangedFloat(Minimum=-3, Maximum=3)) {isSetParameters=true}
   &AddVelocityInCone(
    Cone Angle=10
    Module.Use Velocity Falloff On Cone Axis=false
    Velocity Strength=&UniformRangedFloat(Minimum=100, Maximum=800)
   )
  ]
  particleUpdateScript [
   &ParticleState()
   &SpriteRotationRate(Rotation Rate=&Multiply_Float(
     A=$Particles.InitSpriteRotRate
     B=&FloatFromCurve(FloatCurve=DataInterface() {propertyValues {Curve {keys [<at -0.0046 1.0113 RCIM_Cubic>,<at 1.0011 0.09 RCIM_Cubic>]}}})
    ))
   &Drag(Drag=6, Use Linear Drag=true, Use Rotational Drag=false)
   &SolveForcesAndVelocity(Clamp Velocity=false, Limit Acceleration=false)
   &ScaleSpriteSize(Scale Factor=&Vector2DFromFloat(Value=&FloatFromCurve(
      FloatCurve=DataInterface() {propertyValues {Curve {keys [
          {arriveTangent=2.2362, interpMode=RCIM_Cubic, leaveTangent=2.2362, tangentMode=RCTM_User, time=0, value=0.1}
          {arriveTangent=1.953, interpMode=RCIM_Cubic, leaveTangent=1.953, tangentMode=RCTM_User, time=0.2, value=0.9}
          {arriveTangent=0.0514, interpMode=RCIM_Cubic, leaveTangent=0.0514, tangentMode=RCTM_User, time=1.0008, value=1.6042}
         ]}}}
      Scale Curve=&ClampFloat(
       Float=&FN_LOD_Distance_Remap(Start Distance=0, End Distance=8000, Minimum Output=1, Maximum Output=2)
       Min=1
       Max=2
      )
     )))
   /Game/Effects/Niagara/Modules/Materials/MaterialSmoke(
    Erode:NiagaraFloat=&FloatFromCurve(FloatCurve=DataInterface() {propertyValues {
       bOptimizeLUT=false
       bUseLUT=false
       Curve {keys [
         {arriveTangent=-0.0652, interpMode=RCIM_Cubic, leaveTangent=-0.0652, tangentMode=RCTM_User, time=0.11, value=0.97}
         {arriveTangent=-5.7797, interpMode=RCIM_Cubic, leaveTangent=-5.7797, tangentMode=RCTM_User, time=0.39, value=0.66}
         {arriveTangent=-0.342, interpMode=RCIM_Cubic, leaveTangent=-0.342, tangentMode=RCTM_User, time=0.9954, value=-0.74}
        ]}
      }})
    Shading:NiagaraFloat=&FloatFromCurve(FloatCurve=DataInterface() {propertyValues {
       bOptimizeLUT=false
       bUseLUT=false
       Curve {keys [
         {arriveTangent=0.3846, interpMode=RCIM_Cubic, leaveTangent=0.3846, time=0.07, value=0.5}
         {arriveTangent=0.303, interpMode=RCIM_Cubic, leaveTangent=0.303, time=0.2, value=0.55}
         {interpMode=RCIM_Cubic, tangentMode=RCTM_User, time=0.4, value=0.6}
         {arriveTangent=-1.5, interpMode=RCIM_Cubic, leaveTangent=-1.5, time=0.8, value=0}
        ]}
      }})
   )
   &CameraOffset(Camera Offset Amount=20)
   &DynamicMaterialParameters(
    Index 0 Param 1=0.5
    Index 0 Param 2=&FloatFromCurve(FloatCurve=DataInterface() {propertyValues {Curve {keys [
         {arriveTangent=0.8564, interpMode=RCIM_Cubic, leaveTangent=0.8564, tangentMode=RCTM_User, time=0.0046, value=-0.9753}
         {arriveTangent=4.2876, interpMode=RCIM_Cubic, leaveTangent=5.5268, tangentMode=RCTM_User, time=1, value=1}
        ]}}})
    Index 0 Param 3=&FloatFromCurve(FloatCurve=DataInterface() {propertyValues {Curve {keys [
         {arriveTangent=-3.123, interpMode=RCIM_Cubic, leaveTangent=-3.123, leaveTangentWeight=0.3333, tangentMode=RCTM_User, time=0, value=0.5}
         {arriveTangentWeight=0.1167, interpMode=RCIM_Cubic, leaveTangentWeight=0.2167, tangentMode=RCTM_User, time=0.12, value=0}
        ]}}})
    Index 0 Param 4=&FloatFromCurve(FloatCurve=DataInterface() {propertyValues {Curve {keys [{arriveTangent=-3.2497, interpMode=RCIM_Cubic, leaveTangent=-3.2497, tangentMode=RCTM_User, time=0, value=1}, <at 1 0 RCIM_Cubic>]}}})
    Module.Param0WriteEnabled=true
    Module.Param1WriteEnabled=true
    Module.Param2WriteEnabled=true
    Module.Param3WriteEnabled=true
   )
   &ScaleColor(
    Scale Alpha=&FloatFromCurve(FloatCurve=DataInterface() {propertyValues {Curve {keys [<at -0.003 -0.006>,<at 0.1806 0.5257>,<at 0.7 0.35>,<at 1 0>]}}}, Scale Curve=0.2)
    Scale RGB=&VectorFromFloat(Value=&FloatFromCurve(FloatCurve=DataInterface() {propertyValues {Curve {keys [<at 0 0>,<at 0.1 1.5>,<at 1 0>]}}}))
    ScaleA=true
    ScaleRGB=true
   )
  ]
  rendererClasses [&NiagaraSpriteRendererProperties]
  renderers [
   {rendererClass=&NiagaraSpriteRendererProperties}
   {rendererClass=&NiagaraSpriteRendererProperties, rendererIndex=1}
  ]
 }
 Spikes001 {
  emitterUpdateScript [
   &EmitterState()
   &SpawnBurst_Instantaneous(Spawn Count=8, Use Spawn Probability=false)
  ]
  particleSpawnScript [
   /Niagara/Modules/Spawn/Initialization/InitializeParticle(
    Color=<rgba 10 3.4946 0.2461 1>
    Lifetime=&UniformRangedFloat(Minimum=0.1, Maximum=0.2)
    Mesh Scale=&UniformRangedVector(Minimum=<0.6 0.3 0.3>, Maximum=<0.9 0.8 0.8>)
    Module.Write Color=true
    Module.Write Lifetime=true
    Module.Write Mass=true
    Module.Write Material Random=true
    Module.Write Position=true
    Module.Write Scale=true
    Module.Write SpriteRotation=false
    Module.Write SpriteSize=true
    Position=&Add_Vector(A=&SimulationPosition(), B=<0 0 0>)
    Sprite Size=&RandomRangeVector2D(Minimum=<1 5>, Maximum=<5 40>)
   )
   &AddVelocityInCone(
    Cone Angle=&RandomRangeFloat(Minimum=70, Maximum=100)
    Module.Use Velocity Falloff On Cone Axis=false
    Velocity Strength=&Multiply_Float(
     A=&RandomRangeFloat(Minimum=200, Maximum=700)
     B=&ClampFloat(
      Float=&FN_LOD_Distance_Remap(Start Distance=0, End Distance=8000, Minimum Output=1, Maximum Output=1.5)
      Min=1
      Max=1.5
     )
    )
   )
  ]
  particleUpdateScript [
   &Drag(Drag=2, Use Linear Drag=true, Use Rotational Drag=false)
   &SolveForcesAndVelocity(Clamp Velocity=false, Limit Acceleration=false)
   &ParticleState()
   &Color(
    Color=<rgba 1 0.3636 0.01 1>
    Scale Color=&VectorFromCurve(VectorCurve=DataInterface() {propertyValues {XCurve {keys [<at 0.7 12>,<at 1 1>]}, YCurve {keys [<at 0.7 12>,<at 1 1>]}, ZCurve {keys [<at 0.7 12>,<at 1 1>]}}})
   )
   &ScaleSpriteSize(
    Non-Uniform Curve Scale=&Vector2DFromFloat(Value=&ClampFloat(
      Float=&FN_LOD_Distance_Remap(Start Distance=0, End Distance=8000, Minimum Output=1, Maximum Output=3)
      Min=1
      Max=3
     ))
    Non-Uniform Curve Sprite Scale=DataInterface() {propertyValues {
      XCurve {keys [{tangentMode=RCTM_User, time=0, value=1}, {tangentMode=RCTM_User, time=1, value=0.2}]}
      YCurve {keys [{tangentMode=RCTM_User, time=0, value=1.5}, {tangentMode=RCTM_User, time=1, value=0.7}]}
     }}
    Scale Sprite Size Mode=::NewEnumerator3
   )
  ]
  rendererClasses [&NiagaraSpriteRendererProperties]
  renderers [{rendererClass=&NiagaraSpriteRendererProperties}]
 }
 Spikes_Center {
  emitterUpdateScript [
   &EmitterState()
   &SpawnBurst_Instantaneous(Spawn Count=3, Use Spawn Probability=false)
  ]
  particleSpawnScript [
   &InitializeParticle(
    Color=<rgba 5.3739 3.8765 2.1281 1>
    Lifetime=&UniformRangedFloat(Minimum=0.15, Maximum=0.15)
    Mesh Scale=&UniformRangedVector(Minimum=<1.5 0.1 0.1>, Maximum=<2 0.25 0.25>)
    Module.Write Color=true
    Module.Write Lifetime=true
    Module.Write Mass=true
    Module.Write Material Random=true
    Module.Write Position=true
    Module.Write Scale=true
    Module.Write SpriteRotation=false
    Module.Write SpriteSize=true
    Position=&Add_Vector(A=&SimulationPosition(), B=<0 0 0>)
    Sprite Size=<100 100>
   )
   &CameraOffset(Camera Offset Amount=10)
   &InitialMeshOrientation(
    Mesh Orientation Mode=::NewEnumerator2
    Rotation=&RandomRangeVector(Minimum=<0 -0.2 -0.2>, Maximum=<1 0.2 0.2>)
    Use Rotation Vector=true
   )
  ]
  particleUpdateScript [
   &ParticleState()
   &Color(
    Color=&MakeLinearColorFromVectorAndFloat(
     "Vector (RGB)"=&MakeVectorFromLinearColorRGB(Color=&ColorFromCurve(ColorCurve=DataInterface() {propertyValues {
         AlphaCurve {keys []}
         BlueCurve {keys [
           {arriveTangent=-4.1816, interpMode=RCIM_Cubic, leaveTangent=-4.1816, time=0, value=4.1781}
           {arriveTangent=-4.1816, interpMode=RCIM_Cubic, leaveTangent=-4.1816, time=0.9977, value=0.0059}
          ]}
         GreenCurve {keys [
           {arriveTangent=-7.3723, interpMode=RCIM_Cubic, leaveTangent=-7.3723, time=0, value=7.3615}
           {arriveTangent=-7.3723, interpMode=RCIM_Cubic, leaveTangent=-7.3723, time=0.9977, value=0.0059}
          ]}
         RedCurve {keys [
           {arriveTangent=-14.4701, interpMode=RCIM_Cubic, leaveTangent=-14.4701, time=0, value=15}
           {arriveTangent=-14.4701, interpMode=RCIM_Cubic, leaveTangent=-14.4701, time=0.9977, value=0.5625}
          ]}
        }}))
     "Float (Alpha)"=&FloatFromCurve(FloatCurve=DataInterface() {propertyValues {Curve {keys [{interpMode=RCIM_Cubic, tangentMode=RCTM_User, time=0, value=1}, {interpMode=RCIM_Cubic, tangentMode=RCTM_User, time=1, value=0}]}}})
    )
    Scale Color=&VectorFromFloat(Value=&FloatFromCurve(FloatCurve=DataInterface() {propertyValues {Curve {keys [{interpMode=RCIM_Cubic, tangentMode=RCTM_User, time=0, value=1}, {interpMode=RCIM_Cubic, tangentMode=RCTM_User, time=1, value=0}]}}}))
   )
   &ScaleMeshSize(Scale Factor=&VectorFromCurve(
     VectorCurve=DataInterface() {propertyValues {XCurve {keys [<at 0 0.4>,<at 1 1.2>]}, YCurve {keys [<at 0 1>,<at 1 0>]}, ZCurve {keys [<at 0 1>,<at 1 0>]}}}
     Scale Curve=&VectorFromFloat(Value=&ClampFloat(
       Float=&FN_LOD_Distance_Remap(Start Distance=0, End Distance=8000, Minimum Output=1, Maximum Output=3)
       Min=1
       Max=3
      ))
    ))
   &DynamicMaterialParameters(Write Parameter Index 0=false)
  ]
  rendererClasses [&NiagaraMeshRendererProperties]
  renderers [{rendererClass=&NiagaraMeshRendererProperties, rendererIndex=0}]
 }
}
summary {
 systemName=NS_Core_Sniper_Impact_Default
 emitters [
  {emitterName=Flash, rendererClasses [&NiagaraMeshRendererProperties, &NiagaraSpriteRendererProperties]}
  {emitterName=Smoke, rendererClasses [&NiagaraSpriteRendererProperties]}
  {emitterName=Spikes_Center, rendererClasses [&NiagaraMeshRendererProperties]}
  {emitterName=Debris_Sprite_SubUV, rendererClasses [&NiagaraSpriteRendererProperties]}
  {emitterName=Debris_Mesh_BETA, rendererClasses [&NiagaraMeshRendererProperties]}
  {emitterName=Airflow, rendererClasses [&NiagaraMeshRendererProperties]}
  {emitterName=Spikes001, rendererClasses [&NiagaraSpriteRendererProperties]}
  {emitterName=Light, rendererClasses [&NiagaraLightRendererProperties]}
 ]
 userVariables []
}
